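# Niagara trail effect for the AuraCop glider, combining ribbons, head-wind cones, smoke sprites, spiral meshes and wind-line meshes.
# Ribbon, cone, smoke and spiral emitters burst-spawn in three partitions pinned to skeletal mesh bones, with per-partition time offsets staggering the deploy flare.
# System timelines and smoothed values drive trail stretch, brightness and wind-line rate, while User.RightDot shifts the wind lines.
# VariantIndex blends color variants, and front-end modifiers adjust sizes and intensities.
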NiagaraSystem /BRCosmetics/Gadgets/Assets/VinderTech_GliderChute/Glider_AuraCop/FX/NS_Glider_AuraCop_Trail
dependencies {
 usedDataInterfaces [
  &NiagaraDataInterfaceCurve
  &NiagaraDataInterfaceArrayFloat
  &NiagaraDataInterfaceArrayInt32
  &NiagaraDataInterfaceColorCurve
  &NiagaraDataInterfaceSkeletalMesh
 ]
 usedDynamicInputs [
  &MakeIntFromFloat
  &NormalizeVector
  &TransformVector
 ]
 usedModules [
  &self:SystemSpawnScript.SystemScriptSource.SystemScriptGraph.NiagaraNodeAssignment_0.SetVariables_#1
  &SystemState
  &self:SystemSpawnScript.SystemScriptSource.SystemScriptGraph.NiagaraNodeAssignment_1.SetVariables_#2
  &Timeline
  &FN_Integrator
  &InterpolateOverTime
 ]
}
emitters {
 Local_Ribbon {
  emitterUpdateScript [
   &EmitterState()
   &SpawnBurst_Instantaneous(Spawn Count=30, Use Spawn Probability=false, UseLoopCountLimit=true)
  ]
  particleSpawnScript [
   &PartitionParticles(
    Number Of Partitions=3
    Particles Per Partition=10
    Partition Mode:ENiagara_PartitionMode=::NewEnumerator1
    Write Partition Index to Ribbon Link Order:ENiagara_AttributeSamplingApplyOutput=::NewEnumerator0
    Write Partition to Ribbon ID=::NewEnumerator0
   )
   &InitializeParticle(
    Color=&LerpLinearColors(StartColor=<rgba 1 0.5488 0.1302 1.5>, EndColor=<rgba 1 1 1 1.5>, LerpFactor=$User.VariantIndex)
    Lifetime=0.2
    Position Offset=<0 0 25>
    Ribbon Width=&FN_FrontEndFloatModifier(Gameplay=40, Front End=28)
    Ribbon Width Mode=::NewEnumerator1
    Sprite Size Mode=::NewEnumerator3
    Uniform Sprite Size=5
    UsePositionOffset=false
    Write Material Random=true
   )
   /SetVariables(
    Particles.RibbonID=$Particles.RibbonID
    Particles.RibbonLinkOrder:NiagaraFloat=$Particles.PartitionParticles.Normalized Index In Partition
    Particles.SpawnGroup=$Particles.PartitionParticles.Partition
   ) {isSetParameters=true}
   /SetVariables(Particles.BackSmokeRibbonUOvverride:NiagaraFloat=$Particles.RibbonLinkOrder) {isSetParameters=true}
  ]
  particleUpdateScript [
   /SetVariables(Particles.FadeIn:NiagaraFloat=&FloatFromCurve(
     FloatCurve=DataInterface() {propertyValues {Curve {keys [
         {arriveTangent=2.099, interpMode=RCIM_Cubic, leaveTangent=2.099, tangentMode=RCTM_User, time=0.25, value=0}
         {arriveTangent=-0.489, interpMode=RCIM_Cubic, leaveTangent=-0.489, tangentMode=RCTM_User, time=0.75, value=1}
        ]}}}
     CurveIndex=&Subtract_Float(
      A=&Multiply_Float(A=$System.Age, B=1)
      B=&SelectFloatFromArray(Direct Array Index=$Particles.PartitionParticles.Partition, Float Selection Array=DataInterface() {propertyValues {FloatData [1,0.2,0.7]}})
     )
    )) {isSetParameters=true}
   /SetVariables(Particles.DeployAlpha:NiagaraFloat=&FloatFromCurve(
     FloatCurve=DataInterface() {propertyValues {Curve {keys [
         {arriveTangent=8.1927, interpMode=RCIM_Cubic, leaveTangent=8.1927, tangentMode=RCTM_User, time=0, value=0}
         {interpMode=RCIM_Cubic, tangentMode=RCTM_User, time=0.2, value=1}
         {arriveTangent=-0.489, interpMode=RCIM_Cubic, leaveTangent=-0.489, tangentMode=RCTM_User, time=1.8, value=0}
        ]}}}
     CurveIndex=&Subtract_Float(
      A=&Multiply_Float(A=$System.Age, B=1.5)
      B=&SelectFloatFromArray(Direct Array Index=$Particles.PartitionParticles.Partition, Float Selection Array=DataInterface() {propertyValues {FloatData [1,0.2,0.7]}})
     )
    )) {isSetParameters=true}
   &ParticleState(Kill Particles When Lifetime Has Elapsed=false, Let Infinitely Lived Particles Die When Emitter Deactivates=true)
   &SkeletalMeshLocation(
    Bone / Socket Index=&SelectIntFromArray(Direct Array Index=$Particles.PartitionParticles.Partition, Int Selection Array=DataInterface() {propertyValues {IntData [0,1,2]}})
    Bone Sampling Mode=::NewEnumerator4
    MaskSpawnGroup=false
    Sampled Position Offset=<-90 0 0>
    Skeletal Mesh=$System.Skeletal Mesh
    Use Sampled Mesh Attribute as Ribbon ID=false
   ) {module=SkeletalMeshLocation001}
   /SetVariables(Particles.Position=&LerpPosition(
     A=$Particles.Position
     B=&AddVectorToPosition(
      Position=$Particles.Position
      Vector=&Add_Vector(
       A=&MakeVector(
        X=&FloatFromCurve(
         FloatCurve=DataInterface() {propertyValues {Curve {keys [<at 0 0 RCIM_Cubic>, {arriveTangent=2.1213, interpMode=RCIM_Cubic, leaveTangent=2.1213, tangentMode=RCTM_User, time=1, value=1}]}}}
         CurveIndex=$Particles.RibbonLinkOrder
         Scale Curve=&Multiply_Float(
          A=&MakeFloatFromVector(Vector=$System.InterpolateOverTime.InterpolateOverTime_V3_.Moving Average, Channel=::NewEnumerator0)
          B=&Lerp_Float(
           A=0
           B=&Lerp_Float(A=-35, B=-100, Alpha=$System.Timeline.NormalizedTime)
           Alpha=$System.InterpolateOverTime001.InterpolateOverTime_Float.Moving Average
          )
         )
        )
        Y=0
        Z=0
       )
       B=&Multiply_VectorByFloat(
        Vector=&RotateVectorByQuaternion(Vector To Rotate=<-1 0 0>, Quaternion=$Particles.SkeletalMeshLocation001.SampledRotation)
        Float=&FloatFromCurve(
         FloatCurve=DataInterface() {propertyValues {Curve {keys [{tangentMode=RCTM_User, time=0, value=0}, {tangentMode=RCTM_User, time=1, value=1}]}}}
         CurveIndex=$Particles.RibbonLinkOrder
         Scale Curve=&Lerp_Float(A=&Lerp_Float(A=100, B=200, Alpha=$System.Timeline.NormalizedTime), B=200, Alpha=$Particles.DeployAlpha)
        )
       )
      )
     )
     Alpha=$Particles.RibbonLinkOrder
    )) {isSetParameters=true}
   &ScaleColor(
    Curve Index=$Particles.RibbonLinkOrder
    Linear Color Curve=DataInterface() {propertyValues {
      AlphaCurve {keys [<at 0.0886 1>,<at 0.9269 1>,<at 1 0>]}
      BlueCurve {keys [<at 0 1>,<at 0.9957 0.7031>]}
      GreenCurve {keys [<at 0 1>,<at 0.9957 0.7031>]}
      RedCurve {keys [<at 0 1>,<at 0.9957 0.7031>]}
     }}
    Scale Mode=::NewEnumerator2
   )
   &ScaleColor(
    Scale Alpha=&Multiply_Float(
     A=&FloatFromCurve(
      FloatCurve=DataInterface() {propertyValues {Curve {keys [
          {arriveTangent=0.2105, interpMode=RCIM_Cubic, leaveTangent=0.2105, tangentMode=RCTM_User, time=0, value=1}
          {interpMode=RCIM_Cubic, tangentMode=RCTM_User, time=0.9, value=0}
         ]}}}
      CurveIndex=$Particles.PartitionParticles.Normalized Index In Partition
      Scale Curve=&Lerp_Float(
       A=1.5
       B=2
       Alpha=&Max_Float(A:NiagaraFloat=$System.Timeline.NormalizedTime, B:NiagaraFloat=$Particles.DeployAlpha)
      )
     )
     B=$Particles.FadeIn
    )
    Scale RGB=&MakeVectorFromLinearColorRGB(Color=&ColorFromCurve(
      ColorCurve=DataInterface() {propertyValues {
        AlphaCurve {keys [<at 1 1 RCIM_Cubic>]}
        BlueCurve {keys [<at 0 1 RCIM_Cubic>,<at 1 2.0415 RCIM_Cubic>]}
        GreenCurve {keys [<at 0 1 RCIM_Cubic>,<at 1 3.5437 RCIM_Cubic>]}
        RedCurve {keys [<at 0 1 RCIM_Cubic>,<at 1 5 RCIM_Cubic>]}
       }}
      CurveIndex=&Max_Float(A:NiagaraFloat=$Particles.DeployAlpha, B:NiagaraFloat=$System.Timeline.NormalizedTime)
     ))
    ScaleA=true
    ScaleRGB=true
   ) {module=ScaleColor001}
   &ScaleRibbonWidth(Ribbon Width Scale=&FloatFromCurve(
     FloatCurve=DataInterface() {propertyValues {Curve {keys [{tangentMode=RCTM_User, time=0, value=1}, {tangentMode=RCTM_User, time=1, value=1}]}}}
     CurveIndex=$Particles.Initial.RibbonLinkOrder
    ))
   /SetVariables(Particles.SoftWindColor:LinearColor=&MultiplyLinearColor(Linear Color=$Particles.Color, Scale Factor=<1 1 1 0.2>)) {isSetParameters=true}
   &FN_Sprite_Attributes(
    Color=&MakeLinearColorFromVectorAndFloat(
     "Vector (RGB)"=<1 1 1>
     "Float (Alpha)"=&Multiply_Float(A=&Lerp_Float(A=0, B=1, Alpha=$Particles.DeployAlpha), B=$Particles.FadeIn)
    )
    Create Color Attribute=true
    Create Position Attribute=true
    Position:NiagaraPosition=&AddVectorToPosition(
     Position=$Particles.SkeletalMeshLocation001.SampledPosition
     Vector=&RotateVectorByQuaternion(
      Vector To Rotate=&Lerp_Vector(A=<-75 0 0>, B=$System.FireOffset, Alpha=$Particles.RibbonLinkOrder)
      Quaternion=$Particles.SkeletalMeshLocation001.SampledRotation
     )
    )
    Write to Module Namespace=true
   )
   /SetVariables(Particles.SmokeRibbonWidth:NiagaraFloat=&Multiply_Float(
     A=50
     B=&FloatFromCurve(FloatCurve=DataInterface() {propertyValues {Curve {keys [<at 0 0>,<at 1 1>]}}}, CurveIndex=$Particles.DeployAlpha)
    )) {isSetParameters=true}
   /SetVariables(Particles.RibbonU0Override=&OneMinusFloat(Float=$Particles.RibbonLinkOrder)) {isSetParameters=true}
   &DynamicMaterialParameters(
    Index 1 Param 3=&OneMinusFloat(Float=&FloatFromCurve(FloatCurve=DataInterface() {propertyValues {Curve {keys [<at 0 0>,<at 1 1>]}}}, CurveIndex=$Particles.DeployAlpha))
    Index 3 Param 3=&FloatFromCurve(FloatCurve=DataInterface() {propertyValues {Curve {keys [<at 0 0>,<at 1 1>]}}}, CurveIndex=$Particles.DeployAlpha)
    Module.Index1Float0Write=true
    Module.Index1Float1Write=true
    Module.Index1Float2Write=true
    Module.Index1Float3Write=true
    Module.Index3Float0Write=true
    Module.Index3Float1Write=true
    Module.Index3Float2Write=true
    Module.Index3Float3Write=true
    Write Parameter Index 0=::NewEnumerator0
    Write Parameter Index 1=::NewEnumerator1
    Write Parameter Index 3=::NewEnumerator1
   )
   &FN_Sprite_Attributes(
    Color=&MultiplyLinearColor(
     Linear Color=&LerpLinearColors(StartColor=<rgba 1 1 1 0.15>, EndColor=<rgba 1 1 1 0.25>, LerpFactor=$System.Timeline.NormalizedTime)
     Scale Factor=&MakeFloat4FromVectorAndFloat(
      "Vector (RGB)"=<1 1 1>
      "Float (Alpha)"=&Multiply_Float(
       A=&Multiply_Float(A=&OneMinusFloat(Float=$Particles.DeployAlpha), B=$Particles.FadeIn)
       B=&FloatFromCurve(
        FloatCurve=DataInterface() {propertyValues {Curve {keys [
            {interpMode=RCIM_Cubic, tangentMode=RCTM_User, time=0.7, value=1}
            {arriveTangent=-2.5947, interpMode=RCIM_Cubic, leaveTangent=-2.5947, tangentMode=RCTM_User, time=0.9, value=0}
           ]}}}
        CurveIndex=$Particles.PartitionParticles.Normalized Index In Partition
       )
      )
     )
    )
    Create Color Attribute=true
    Write to Module Namespace=true
   ) {module=FN_Sprite_Attributes001}
  ]
  rendererClasses [&NiagaraRibbonRendererProperties]
  renderers [
   {rendererClass=&NiagaraRibbonRendererProperties}
   {rendererClass=&NiagaraRibbonRendererProperties, rendererIndex=1}
   {rendererClass=&NiagaraRibbonRendererProperties, rendererIndex=2}
   {rendererClass=&NiagaraRibbonRendererProperties, rendererIndex=3}
  ]
 }
 NE_HeadWindCone {
  emitterUpdateScript [
   &EmitterState(
    Inactive Response=::NewEnumerator1
    Life Cycle Mode=::NewEnumerator1
    Loop Behavior=::NewEnumerator1
    Loop Duration Mode=::NewEnumerator1
    UseLoopDelay=false
   )
   &SpawnBurst_Instantaneous(Spawn Count=3, Use Spawn Probability=false, UseLoopCountLimit=false)
  ]
  particleSpawnScript [
   &PartitionParticles(ID to Partition:NiagaraInt32=&ReturnExecIndex(), Number Of Partitions=3)
   &InitializeParticle(
    Color=<rgba 1 1 1 5>
    Mesh Scale=<1.25 0.45 0.45>
    Mesh Scale Mode=::NewEnumerator4
    Position Offset Coordinate Space=::Local
    Sprite Size=&FN_FrontEndVector2Modifier(Front End:Vector2f=<150 330>, Gameplay:Vector2f=<150 370>)
    Sprite Size Mode=::NewEnumerator4
    UsePositionOffset=true
    Write Material Random=true
   )
   &InitialMeshOrientation(Orientation Coordinate Space=::Local, Rotation Coordinate Space=::NewEnumerator2, Use Orientation Vector=true, Use Rotation Vector=true)
   &CameraOffset(Camera Offset Amount=10)
  ]
  particleUpdateScript [
   /SetVariables(Particles.SheenProgress:NiagaraFloat=&FloatFromCurve(
     FloatCurve=DataInterface() {propertyValues {Curve {keys [{arriveTangent=2.7245, time=0, value=0}, {arriveTangent=-0.489, leaveTangent=-0.489, time=0.3, value=1}]}}}
     CurveIndex=&Subtract_Float(
      A=&Multiply_Float(A=$System.Age, B=1)
      B=&SelectFloatFromArray(Direct Array Index=$Particles.PartitionParticles.Partition, Float Selection Array=DataInterface() {propertyValues {FloatData [1,0.45,0.75]}})
     )
    )) {isSetParameters=true}
   &ParticleState(Kill Particles When Lifetime Has Elapsed=false)
   /SetVariables(Particles.FadeIn:NiagaraFloat=&FloatFromCurve(
     FloatCurve=DataInterface() {propertyValues {Curve {keys [
         {arriveTangent=2.099, interpMode=RCIM_Cubic, leaveTangent=2.099, tangentMode=RCTM_User, time=0, value=0}
         {arriveTangent=-0.489, interpMode=RCIM_Cubic, leaveTangent=-0.489, tangentMode=RCTM_User, time=0.1, value=1}
        ]}}}
     CurveIndex=&Subtract_Float(
      A=&Multiply_Float(A=$System.Age, B=1)
      B=&SelectFloatFromArray(Direct Array Index=$Particles.PartitionParticles.Partition, Float Selection Array=DataInterface() {propertyValues {FloatData [0.9,0.3,0.7]}})
     )
    )) {isSetParameters=true}
   /SetVariables(Particles.DeployAlpha:NiagaraFloat=&FloatFromCurve(
     FloatCurve=DataInterface() {propertyValues {Curve {keys [
         {arriveTangent=8.1927, interpMode=RCIM_Cubic, leaveTangent=8.1927, tangentMode=RCTM_User, time=0, value=0}
         {interpMode=RCIM_Cubic, tangentMode=RCTM_User, time=0.2, value=1}
         {arriveTangent=-0.489, interpMode=RCIM_Cubic, leaveTangent=-0.489, tangentMode=RCTM_User, time=1, value=0}
        ]}}}
     CurveIndex=&Subtract_Float(
      A=&Multiply_Float(A=$System.Age, B=1)
      B=&SelectFloatFromArray(Direct Array Index=$Particles.PartitionParticles.Partition, Float Selection Array=DataInterface() {propertyValues {FloatData [1,0.2,0.7]}})
     )
    )) {isSetParameters=true}
   &SkeletalMeshLocation(
    Apply Sampled Orientation to Mesh Particles=true
    Bone / Socket Index=$Particles.PartitionParticles.Partition
    Bone Sampling Mode=::NewEnumerator4
    Inherit Velocity=false
    MaskSpawnGroup=false
    Sampled Position Offset Coordinate Space:ENiagaraMeshTransforms=::NewEnumerator3
    Skeletal Mesh=$System.Skeletal Mesh
   )
   &ScaleColor(
    Scale Alpha=&Multiply_Float(
     A=&Max_Float(
      A:NiagaraFloat=$Particles.DeployAlpha
      B:NiagaraFloat=&Multiply_Float(A=$System.Timeline.NormalizedTime, B=0.5)
     )
     B=$Particles.FadeIn
    )
    Scale RGB=&MakeVectorFromLinearColorRGB(Color=&LerpLinearColors(
      StartColor=&ColorFromCurve(
       ColorCurve=DataInterface() {propertyValues {
         AlphaCurve {keys [<at 0 1 RCIM_Cubic>]}
         BlueCurve {keys [<at 0 0.2248 RCIM_Cubic>,<at 1 0 RCIM_Cubic>]}
         GreenCurve {keys [<at 0 0.615 RCIM_Cubic>,<at 1 1.4975 RCIM_Cubic>]}
         RedCurve {keys [<at 0 1 RCIM_Cubic>,<at 1 5 RCIM_Cubic>]}
        }}
       CurveIndex=$Particles.DeployAlpha
      )
      EndColor=&ColorFromCurve(
       ColorCurve=DataInterface() {propertyValues {
         AlphaCurve {keys [<at 0 1 RCIM_Cubic>]}
         BlueCurve {keys [<at 0 1 RCIM_Cubic>,<at 1 0 RCIM_Cubic>]}
         GreenCurve {keys [<at 0 1 RCIM_Cubic>,<at 1 2.5578 RCIM_Cubic>]}
         RedCurve {keys [<at 0 1 RCIM_Cubic>,<at 1 5 RCIM_Cubic>]}
        }}
       CurveIndex=$Particles.DeployAlpha
      )
      LerpFactor=$User.VariantIndex
     ))
    ScaleA=true
    ScaleRGB=true
   )
   &DynamicMaterialParameters(
    Index 2 Param 1=&Lerp_Float(A=-1, B=2, Alpha=$Particles.SheenProgress)
    Index 3 Param 1=3
    Index 3 Param 2=20
    Index 3 Param 3=1
    Module.Index2Float0Write=true
    Module.Index2Float1Write=true
    Module.Index2Float2Write=true
    Module.Index2Float3Write=true
    Module.Index3Float0Write=true
    Module.Index3Float1Write=true
    Module.Index3Float2Write=true
    Module.Index3Float3Write=true
    Write Parameter Index 0=::NewEnumerator0
    Write Parameter Index 2=::NewEnumerator1
    Write Parameter Index 3=::NewEnumerator1
   )
   &SpriteFacingAndAlignment(
    Sprite Alignment=&RotateVectorByQuaternion(Vector To Rotate=<1 0 0>, Quaternion=$Output.SkeletalMeshLocation.SampledRotation)
    WriteAlignment=true
    WriteFacing=false
   )
   &FN_Sprite_Attributes(
    Color=&Multiply_LinearColorByFloat(
     Linear Color=&LerpLinearColors(
      StartColor=&ColorFromCurve(
       ColorCurve=DataInterface() {propertyValues {
         AlphaCurve {keys [<at 0 0.4 RCIM_Cubic>,<at 0.9813 1>]}
         BlueCurve {keys [<at 0 0.1559 RCIM_Cubic>,<at 1 0 RCIM_Cubic>]}
         GreenCurve {keys [<at 0 0.7451 RCIM_Cubic>,<at 1 1.2348 RCIM_Cubic>]}
         RedCurve {keys [<at 0 1 RCIM_Cubic>,<at 1 3 RCIM_Cubic>]}
        }}
       CurveIndex=&Multiply_Float(
        A=&Max_Float(
         A:NiagaraFloat=$Particles.DeployAlpha
         B:NiagaraFloat=&Multiply_Float(A=$System.Timeline.NormalizedTime, B=0.75)
        )
        B=&FN_FrontEndFloatModifier(Gameplay=1, Front End=0.5)
       )
      )
      EndColor=&ColorFromCurve(
       ColorCurve=DataInterface() {propertyValues {
         AlphaCurve {keys [<at 0 0.4 RCIM_Cubic>,<at 0.9813 1>]}
         BlueCurve {keys [<at 0 1 RCIM_Cubic>,<at 1 1.1562 RCIM_Cubic>]}
         GreenCurve {keys [<at 0 1 RCIM_Cubic>,<at 1 1.5525 RCIM_Cubic>]}
         RedCurve {keys [<at 0 1 RCIM_Cubic>,<at 1 2 RCIM_Cubic>]}
        }}
       CurveIndex=&Multiply_Float(
        A=&Max_Float(
         A:NiagaraFloat=$Particles.DeployAlpha
         B:NiagaraFloat=&Multiply_Float(A=$System.Timeline.NormalizedTime, B=0.75)
        )
        B=&FN_FrontEndFloatModifier(Gameplay=1, Front End=0.5)
       )
      )
      LerpFactor=$User.VariantIndex
     )
     Float=$Particles.FadeIn
    )
    Create Color Attribute=true
    Create Position Attribute=true
    Position:NiagaraPosition=&AddVectorToPosition(
     Position=$Output.SkeletalMeshLocation.SampledPosition
     Vector=&RotateVectorByQuaternion(
      Vector To Rotate=&FN_FrontEndVector3Modifier(Gameplay=<-10 0 0>, Front End=<-10 0 0>)
      Quaternion=$Output.SkeletalMeshLocation.SampledRotation
     )
    )
    Write to Module Namespace=true
   )
   &FN_Mesh_Attributes(
    Color=&Multiply_LinearColorByFloat(
     Linear Color=$Particles.FN_Sprite_Attributes.Color
     Float=&Multiply_Float(
      A=2
      B=&FloatFromCurve(
       FloatCurve=DataInterface() {propertyValues {Curve {keys [
           {arriveTangent=6.0701, interpMode=RCIM_Cubic, leaveTangent=6.0701, tangentMode=RCTM_User, time=0, value=0}
           {arriveTangent=0.0353, interpMode=RCIM_Cubic, leaveTangent=0.0353, tangentMode=RCTM_User, time=0.3, value=1}
           {arriveTangent=-0.3611, interpMode=RCIM_Cubic, leaveTangent=-0.3611, tangentMode=RCTM_User, time=1, value=0}
          ]}}}
       CurveIndex=$Particles.SheenProgress
      )
     )
    )
    Create Color Attribute=true
    Create Orientation Attribute=true
    Create Scale Attribute=true
    Orientation:Quat4f=$Output.SkeletalMeshLocation.SampledRotation
    Scale=<1 1 1>
    Write to Module Namespace=true
   )
  ]
  rendererClasses [&NiagaraMeshRendererProperties, &NiagaraSpriteRendererProperties]
  renderers [
   {rendererClass=&NiagaraMeshRendererProperties, rendererIndex=0}
   {rendererClass=&NiagaraSpriteRendererProperties, rendererIndex=1}
   {rendererClass=&NiagaraMeshRendererProperties, rendererIndex=2}
  ]
 }
 NE_SmokeTrail {
  emitterSpawnScript [/SetVariables(Emitter.SmokeSpritePerBullet=9) {isSetParameters=true}]
  emitterUpdateScript [
   &EmitterState(
    Inactive Response=::NewEnumerator1
    Life Cycle Mode=::NewEnumerator1
    Loop Behavior=::NewEnumerator1
    Loop Duration Mode=::NewEnumerator1
    UseLoopDelay=false
   )
   &SpawnBurst_Instantaneous(
    Spawn Count=&Multiply_Int(A=3, B=$Emitter.SmokeSpritePerBullet)
    Use Spawn Probability=false
    UseLoopCountLimit=false
   )
  ]
  particleSpawnScript [
   &PartitionParticles(
    Number Of Partitions=3
    Particles Per Partition:NiagaraInt32=$Emitter.SmokeSpritePerBullet
    Partition Mode:ENiagara_PartitionMode=::NewEnumerator1
    Write Partition Index to Ribbon Link Order:ENiagara_AttributeSamplingApplyOutput=::NewEnumerator0
    Write Partition to Ribbon ID=::NewEnumerator0
   )
   &InitializeParticle(
    Color=<rgba 0.0677 0.0458 0.017 2>
    Lifetime=&Lerp_Float(A=1, B=0.25, Alpha=$Particles.PartitionParticles.Normalized Index In Partition)
    Sprite Size Mode=::NewEnumerator3
    Uniform Sprite Size=50
    UsePositionOffset=false
    Write Material Random=true
   )
   /SetVariables(Particles.Age:NiagaraFloat=&Subtract_Float(
     A=&SelectFloatFromArray(
      Direct Array Index=$Particles.PartitionParticles.Partition
      Float Selection Array=DataInterface() {propertyValues {FloatData [-0.8,-0.2,-0.6]}}
     )
     B=&Multiply_Float(A=$Particles.PartitionParticles.Normalized Index In Partition, B=0.3)
    )) {isSetParameters=true}
   &AddVelocity(Velocity=<0 -1 0>, Velocity Speed Scale=&RandomRangeFloat(Minimum=180, Maximum=250)) {enabled=false}
  ]
  particleUpdateScript [
   &SkeletalMeshLocation(
    Bone / Socket Index=&SelectIntFromArray(Direct Array Index=$Particles.PartitionParticles.Partition, Int Selection Array=DataInterface() {propertyValues {IntData [0,1,2]}})
    Bone Sampling Mode=::NewEnumerator4
    MaskSpawnGroup=false
    Sampled Position Offset=$System.FireOffset
    Skeletal Mesh=$System.Skeletal Mesh
   ) {module=SkeletalMeshLocation001}
   /SetVariables(Particles.Position=&Lerp_Vector(
     A=&AddVectorToPosition(
      Position=$Particles.SkeletalMeshLocation001.SampledPosition
      Vector=&RotateVectorByQuaternion(Vector To Rotate=$System.FireOffset, Quaternion=$Particles.SkeletalMeshLocation001.SampledRotation)
     )
     B=&AddVectorToPosition(
      Position=$Particles.SkeletalMeshLocation001.SampledPosition
      Vector=&RotateVectorByQuaternion(Vector To Rotate=<-125 0 0>, Quaternion=$Particles.SkeletalMeshLocation001.SampledRotation)
     )
     Alpha=$Particles.PartitionParticles.Normalized Index In Partition
    )) {isSetParameters=true}
   &ParticleState()
   &ScaleSpriteSize(
    Uniform Curve Scale=&Lerp_Float(
     A=&FN_FrontEndFloatModifier(Gameplay=2, Front End=1)
     B=0.75
     Alpha=$Particles.PartitionParticles.Normalized Index In Partition
    )
    Uniform Curve Sprite Scale=DataInterface() {propertyValues {Curve {keys [{tangentMode=RCTM_User, time=0, value=0}, {tangentMode=RCTM_User, time=0.1, value=1}]}}}
   )
   &SolveForcesAndVelocity(Clamp Velocity=false, Limit Acceleration=false)
   &DynamicMaterialParameters(
    Index 0 Param 1=&Add_Float(
     A=&RandomRangeFloat(Minimum=0, Maximum=1)
     B=&FloatFromCurve(FloatCurve=DataInterface() {propertyValues {Curve {keys [
          {arriveTangent=2.3076, interpMode=RCIM_Cubic, leaveTangent=2.3076, tangentMode=RCTM_User, time=0, value=1.1404}
          {arriveTangent=0.2537, interpMode=RCIM_Cubic, leaveTangent=0.2537, tangentMode=RCTM_User, time=1, value=2}
         ]}}})
    )
    Index 0 Param 2=&FloatFromCurve(FloatCurve=DataInterface() {propertyValues {Curve {keys [
         {arriveTangent=0.0453, interpMode=RCIM_Cubic, tangentMode=RCTM_Break, time=0.3817, value=-1}
         {arriveTangent=0.0604, interpMode=RCIM_Cubic, leaveTangent=0.0604, tangentMode=RCTM_User, time=1, value=-1}
        ]}}})
    Index 0 Param 3=&FloatFromCurve(
     FloatCurve=DataInterface() {propertyValues {Curve {keys [
         {arriveTangent=-10.301, interpMode=RCIM_Cubic, leaveTangent=-4.5181, tangentMode=RCTM_User, time=0, value=0.5}
         {interpMode=RCIM_Cubic, leaveTangent=4.6779, time=0.258, value=0}
        ]}}}
     CurveIndex=&Multiply_Float(A=$Particles.NormalizedAge, B=1)
     Scale Curve=&Multiply_Float(
      A=150
      B=&FloatFromCurve(
       FloatCurve=DataInterface() {propertyValues {Curve {keys [
           {arriveTangent=-2.9452, interpMode=RCIM_Cubic, leaveTangent=-2.9452, tangentMode=RCTM_User, time=0, value=1}
           {interpMode=RCIM_Cubic, tangentMode=RCTM_User, time=1, value=0}
          ]}}}
       CurveIndex=$Particles.PartitionParticles.Normalized Index In Partition
      )
     )
    )
    Index 0 Param 4=&Lerp_Float(A=1, B=0, Alpha=$Particles.NormalizedAge)
    Index 1 Param 1=&FloatFromCurve(FloatCurve=DataInterface() {propertyValues {Curve {keys [<at 0 0>,<at 0.05 1>]}}})
    Index 1 Param 2=&Add_Float(
     A=&MakeFloatFromVector4(Vector 4=$Particles.DynamicMaterialParameter1, Channel=::NewEnumerator1)
     B=&Multiply_Float(A=$Particles.NormalizedAge, B=0.05)
    )
    Index 1 Param 3=&FloatFromCurve(FloatCurve=DataInterface() {propertyValues {Curve {keys [
         {interpMode=RCIM_Cubic, leaveTangent=2.8445, tangentMode=RCTM_Break, time=0, value=0.2}
         {interpMode=RCIM_Cubic, leaveTangent=63.2978, time=0.05, value=1.5}
        ]}}})
    Index 1 Param 4=&FloatFromCurve(
     FloatCurve=DataInterface() {propertyValues {Curve {keys [<at 0 1>,<at 1 0.8>]}}}
     Scale Curve=&RandomRangeFloat(Minimum=0.9, Maximum=1)
    )
    Index 2 Param 1=&FloatFromCurve(FloatCurve=DataInterface() {propertyValues {Curve {keys [{interpMode=RCIM_Cubic, leaveTangent=3.2309, tangentMode=RCTM_Break, time=0, value=0}, <at 1 1.25>]}}})
    Index 2 Param 2=&FloatFromCurve(
     FloatCurve=DataInterface() {propertyValues {Curve {keys [
         {interpMode=RCIM_Cubic, leaveTangent=3.7422, tangentMode=RCTM_Break, time=0, value=0.2}
         {interpMode=RCIM_Cubic, leaveTangent=0.9066, time=1, value=2}
        ]}}}
     Scale Curve=&RandomRangeFloat(Minimum=0.2, Maximum=1.2)
    )
    Index 2 Param 3=&FloatFromCurve(FloatCurve=DataInterface() {propertyValues {Curve {keys [<at 0 0.5 RCIM_Cubic>, {interpMode=RCIM_Cubic, leaveTangent=11.1233, time=1, value=1}]}}})
    Index 2 Param 4=&FloatFromCurve(
     FloatCurve=DataInterface() {propertyValues {Curve {keys [
         {arriveTangent=-1.3655, interpMode=RCIM_Cubic, leaveTangent=-1.3655, tangentMode=RCTM_User, time=0, value=1}
         {interpMode=RCIM_Cubic, leaveTangent=-1.2956, tangentMode=RCTM_Break, time=1, value=0.45}
        ]}}}
     Scale Curve=2
    )
    Module.Param0WriteEnabled=true
    Module.Param0WriteEnabled1=true
    Module.Param0WriteEnabled2=true
    Module.Param1WriteEnabled=true
    Module.Param1WriteEnabled1=true
    Module.Param1WriteEnabled2=true
    Module.Param2WriteEnabled=true
    Module.Param2WriteEnabled1=true
    Module.Param2WriteEnabled2=true
    Module.Param3WriteEnabled=true
    Module.Param3WriteEnabled1=true
    Module.Param3WriteEnabled2=true
    Write Parameter Index 1=true
    Write Parameter Index 2=true
   ) {module=DynamicMaterialParameters001}
   &ScaleColor(
    Scale Alpha=&FloatFromCurve(
     FloatCurve=DataInterface() {propertyValues {Curve {keys [
         {interpMode=RCIM_Cubic, tangentMode=RCTM_User, time=0, value=0}
         {interpMode=RCIM_Cubic, tangentMode=RCTM_User, time=0.35, value=1}
         {interpMode=RCIM_Cubic, tangentMode=RCTM_User, time=1, value=0}
        ]}}}
     Scale Curve=0.9
    )
    ScaleA=true
    ScaleRGB=true
   )
  ]
  rendererClasses [&NiagaraSpriteRendererProperties]
  renderers [
   {rendererClass=&NiagaraSpriteRendererProperties}
   {rendererClass=&NiagaraSpriteRendererProperties, rendererIndex=1}
  ]
 }
 NE_SpiralTrails001 {
  emitterSpawnScript [/SetVariables(Emitter.SmokeSpritePerBullet=5) {isSetParameters=true}]
  emitterUpdateScript [
   &EmitterState(
    Inactive Response=::NewEnumerator1
    Life Cycle Mode=::NewEnumerator1
    Loop Behavior=::NewEnumerator1
    Loop Duration Mode=::NewEnumerator1
    UseLoopDelay=false
   )
   &SpawnBurst_Instantaneous(
    Spawn Count=&Multiply_Int(A=3, B=$Emitter.SmokeSpritePerBullet)
    Use Spawn Probability=false
    UseLoopCountLimit=false
   )
  ]
  particleSpawnScript [
   &PartitionParticles(
    Number Of Partitions=3
    Particles Per Partition:NiagaraInt32=$Emitter.SmokeSpritePerBullet
    Partition Mode:ENiagara_PartitionMode=::NewEnumerator1
   )
   /SetVariables(Particles.Age:NiagaraFloat=&Subtract_Float(
     A=&SelectFloatFromArray(
      Direct Array Index=$Particles.PartitionParticles.Partition
      Float Selection Array=DataInterface() {propertyValues {FloatData [-0.8,-0.2,-0.6]}}
     )
     B=&Multiply_Float(A=$Particles.PartitionParticles.Normalized Index In Partition, B=0.5)
    )) {isSetParameters=true}
   &InitializeParticle(
    Lifetime=&Lerp_Float(A=0.8, B=0.6, Alpha=$Particles.PartitionParticles.Normalized Index In Partition)
    Mesh Scale Mode=::NewEnumerator3
    Mesh Uniform Scale=&Multiply_Float(A=1.5, B=&FN_FrontEndFloatModifier(Gameplay=1, Front End=0.7))
    Sprite Size Mode=::NewEnumerator3
    UsePositionOffset=true
    Write Material Random=true
   )
   &InitialMeshOrientation(
    Mesh Orientation Mode=::NewEnumerator7
    Rotation=&RandomRangeVector(Minimum=<0 0 0>, Maximum=<0 1 0>)
    Use Rotation Vector=true
   )
   /SetVariables(Particles.GlowDecalSize:Vector2f=&Multiply_Vector2D_ByFloat(Float=&Lerp_Float(A=250, B=100, Alpha=$System.ProgressAlpha), Vector2D=<1 1>)) {isSetParameters=true}
   &AddVelocity(Velocity=&MaskVectorBySpawnGroup(Masked Vector=<0 0 0>, Passthrough Vector=&RandomRangeVector(Minimum=<0 -50 0>, Maximum=<0 -75 0>)))
  ]
  particleUpdateScript [
   &ParticleState()
   &SkeletalMeshLocation(
    Bone / Socket Index=&SelectIntFromArray(Direct Array Index=$Particles.PartitionParticles.Partition, Int Selection Array=DataInterface() {propertyValues {IntData [0,1,2]}})
    Bone Sampling Mode=::NewEnumerator4
    MaskSpawnGroup=false
    Orientation Sampling=::NewEnumerator1
    Sampled Position Offset=$System.FireOffset
    Skeletal Mesh=$System.Skeletal Mesh
   ) {module=SkeletalMeshLocation001}
   /SetVariables(Particles.Position=&Lerp_Vector(
     A=&AddVectorToPosition(
      Position=$Particles.SkeletalMeshLocation001.SampledPosition
      Vector=&RotateVectorByQuaternion(Vector To Rotate=$System.FireOffset, Quaternion=$Particles.SkeletalMeshLocation001.SampledRotation)
     )
     B=&AddVectorToPosition(
      Position=$Particles.SkeletalMeshLocation001.SampledPosition
      Vector=&RotateVectorByQuaternion(Vector To Rotate=<-100 0 0>, Quaternion=$Particles.SkeletalMeshLocation001.SampledRotation)
     )
     Alpha=$Particles.PartitionParticles.Normalized Index In Partition
    )) {isSetParameters=true}
   &ScaleMeshSize(Scale Factor=&VectorFromFloat(Value=&FloatFromCurve(
      FloatCurve=DataInterface() {propertyValues {Curve {keys [
          {arriveTangent=2.4521, interpMode=RCIM_Cubic, leaveTangent=2.4521, tangentMode=RCTM_User, time=0, value=0}
          {interpMode=RCIM_Cubic, tangentMode=RCTM_User, time=1, value=1}
         ]}}}
      Scale Curve=&Lerp_Float(
       A=&FN_FrontEndFloatModifier(Gameplay=0.7, Front End=1)
       B=0.25
       Alpha=$Particles.PartitionParticles.Normalized Index In Partition
      )
     )))
   &Color(Color=&Multiply_LinearColorByFloat(
     Linear Color=&ColorFromCurve(ColorCurve=DataInterface() {propertyValues {
        AlphaCurve {keys [<at 0 0.75 RCIM_Cubic>,<at 1 0 RCIM_Cubic>]}
        BlueCurve {keys [<at 0 0 RCIM_Cubic>,<at 0.2409 1>,<at 0.7276 1 RCIM_Cubic>]}
        GreenCurve {keys [<at 0 4.0082 RCIM_Cubic>,<at 0.2409 1>,<at 0.7276 1 RCIM_Cubic>]}
        RedCurve {keys [<at 0 10 RCIM_Cubic>,<at 0.2409 1>,<at 0.7276 1 RCIM_Cubic>]}
       }})
     Float=&FN_FrontEndFloatModifier(Gameplay=1, Front End=0.5)
    ))
   &SolveForcesAndVelocity(Clamp Velocity=false, Limit Acceleration=false)
   &DynamicMaterialParameters(
    Index 1 Param 3=&FloatFromCurve(FloatCurve=DataInterface() {propertyValues {Curve {keys [<at 0 0>,<at 1 1>]}}})
    Module.Index0Float0Write=true
    Module.Index0Float1Write=true
    Module.Index0Float2Write=true
    Module.Index0Float3Write=true
    Module.Index1Float0Write=true
    Module.Index1Float1Write=true
    Module.Index1Float2Write=true
    Module.Index1Float3Write=true
    Write Parameter Index 1=::NewEnumerator1
   )
  ]
  rendererClasses [&NiagaraMeshRendererProperties]
  renderers [{rendererClass=&NiagaraMeshRendererProperties, rendererIndex=0}]
 }
 WindLines {
  emitterUpdateScript [
   &EmitterState()
   &SpawnRate(
    SpawnRate=&Lerp_Float(
     A=50
     B=150
     Alpha=&RemapRange(Input Value To Remap=$User.RightDot, Input Min=-1, Input Max=1, Output Min=0, Output Max=1)
    )
    Use Spawn Probability=true
   ) {enabled=false}
   &SpawnRate(
    SpawnGroup=1
    SpawnRate=&Lerp_Float(
     A=&FN_FrontEndFloatModifier(
      Gameplay=0
      Front End=&FloatFromCurve(FloatCurve=DataInterface() {propertyValues {Curve {keys [<at 0.5 0>,<at 2 1>]}}}, CurveIndex=$Engine.System.Age, Scale Curve=200)
     )
     B=200
     Alpha=$System.Timeline.NormalizedTime
    )
    Use Spawn Probability=true
   ) {module=SpawnRate001}
  ]
  particleSpawnScript [
   &InitializeParticle(
    Color=&LerpLinearColors(StartColor=<rgba 1 0.8756 0.4483 5>, EndColor=<rgba 1 1 1 5>, LerpFactor=$User.VariantIndex)
    Mesh Scale Max=<25 1 1>
    Mesh Scale Min=<10 1 1>
    Mesh Scale Mode=::NewEnumerator2
    UsePositionOffset=false
    Write Material Random=true
   )
   &InitialMeshOrientation(
    Mesh Orientation Mode=::NewEnumerator2
    Rotation=&RandomRangeVector(Minimum=<0 0 0>, Maximum=<1 0 0>)
    Use Rotation Vector=true
   )
   &TorusLocation(
    Align Torus With Axis=true
    Handle Radius=0
    Large Radius:NiagaraFloat=&RandomRangeFloat(Minimum=80, Maximum=150)
    MaskSpawnGroup=false
    Offset=<0 -180 0>
    Torus Axis=<0 1 0>
    Torus Distribution Mode:ENiagaraTorusDistributionMode=::NewEnumerator1
    Torus Origin:NiagaraPosition=&Add_Vector(
     A=Linked() {linkedVariable {name=Particles.Position, type=NiagaraPosition}}
     B=&MakeVector(X=0, Y=&MakeCustomFloatFromBool(Bool=$User.bIsFrontEndPreview, True Float=200, False Float=0), Z=0)
    )
    U Position:NiagaraFloat=&MakeCustomFloatFromBool(
     Bool=&SetBoolByFloatComparison(A=$User.RightDot, B=0, Comparison Type=::NewEnumerator4)
     True Float=&RandomRangeFloat(Minimum=0, Maximum=1)
     False Float=&Lerp_Float(
      A=&RandomRangeFloat(Minimum=0.25, Maximum=0.75)
      B=&RandomRangeFloat(Minimum=-0.25, Maximum=0.25)
      Alpha=&RemapRange(Input Value To Remap=$User.RightDot, Input Min=-1, Input Max=1, Output Min=0, Output Max=1)
     )
    )
    V Position:NiagaraFloat=&RandomRangeFloat(Minimum=0, Maximum=1)
   ) {enabled=false}
   &TorusLocation(
    Align Torus With Axis=true
    Handle Radius=0
    Large Radius:NiagaraFloat=&FN_FrontEndFloatModifier(
     Gameplay=&RandomRangeFloat(Minimum=180, Maximum=220)
     Front End=&RandomRangeFloat(Minimum=200, Maximum=250)
    )
    MaskSpawnGroup=false
    Module.Override Randomness=false
    Offset=<0 -180 0>
    Spawn Group Mask=1
    Torus Axis=<0 1 0>
    Torus Origin:NiagaraPosition=&Add_Vector(
     A=&FN_FrontEndVector3Modifier(Gameplay=<0 -100 50>, Front End=<0 -500 50>)
     B=&Lerp_Vector(A=<0 0 0>, B=<200 0 0>, Alpha=$User.RightDot)
    )
   ) {module=TorusLocation001}
   &AddVelocityFromPoint(
    Module.Use Falloff Distance=false
    Origin Offset=&Add_Vector(A=<0 -500 0>, B=&Lerp_Vector(A=<0 0 0>, B=<300 0 0>, Alpha=$User.RightDot))
    Velocity Strength=&FN_FrontEndFloatModifier(Gameplay=150, Front End=-150)
   )
   &AddVelocity(
    Module.Use Falloff Distance=false
    Origin Offset=&Add_Vector(A=<0 -500 0>, B=&Lerp_Vector(A=<0 0 0>, B=<500 0 0>, Alpha=$User.RightDot))
    Override Seed=false
    Velocity Mode=::NewEnumerator1
    Velocity Speed=200
   ) {module=AddVelocity001}
   &AddVelocity(Velocity=&Multiply_Vector(
     A=&RandomRangeVector(Minimum=<1 400 1>, Maximum=<1 600 1>)
     B=&MakeVector(X=1, Y=&MakeCustomFloatFromBool(Bool=$User.bIsFrontEndPreview, True Float=5, False Float=1), Z=1)
    ))
  ]
  particleUpdateScript [
   &ScaleVelocity(Velocity Scale=&VectorFromFloat(Value=&FloatFromCurve(
      FloatCurve=DataInterface() {propertyValues {Curve {keys [
          {interpMode=RCIM_Cubic, tangentMode=RCTM_User, time=0, value=0}
          {arriveTangent=2, interpMode=RCIM_Cubic, leaveTangent=2, time=0.1, value=0.975}
          {interpMode=RCIM_Cubic, tangentMode=RCTM_User, time=0.5, value=1}
          {arriveTangent=-2, interpMode=RCIM_Cubic, leaveTangent=-2, time=0.9, value=1}
          {tangentMode=RCTM_User, time=1, value=0}
         ]}}}
      CurveIndex=&RemapRange(Input Value To Remap=$User.RightDot, Input Min=-1, Input Max=1, Output Min=0, Output Max=1)
     )))
   &SolveForcesAndVelocity(Clamp Velocity=false, Limit Acceleration=false)
   &ParticleState()
   &ScaleMeshSize(Scale Factor=&VectorFromFloat(Value=&FloatFromCurve(FloatCurve=DataInterface() {propertyValues {Curve {keys [
          {tangentMode=RCTM_User, time=0, value=1}
          {arriveTangent=-0.0405, interpMode=RCIM_Cubic, leaveTangent=-0.0405, tangentMode=RCTM_User, time=0.7009, value=1}
          {arriveTangent=-8.0089, interpMode=RCIM_Cubic, leaveTangent=-8.0089, tangentMode=RCTM_User, time=1, value=0}
         ]}}})))
   &DynamicMaterialParameters(
    Index 0 Param 2=1
    Index 0 Param 3=1
    Index 0 Param 4=1
    Module.Param0WriteEnabled=true
    Module.Param1WriteEnabled=true
    Module.Param2WriteEnabled=true
    Module.Param3WriteEnabled=true
   )
   &ScaleColor(
    Scale Alpha=&FloatFromCurve(
     FloatCurve=DataInterface() {propertyValues {Curve {keys [
         {interpMode=RCIM_Cubic, tangentMode=RCTM_User, time=0, value=0}
         {interpMode=RCIM_Cubic, tangentMode=RCTM_User, time=0.5, value=1.5}
         {tangentMode=RCTM_User, time=1, value=0}
        ]}}}
     Scale Curve=&Lerp_Float(
      A=&FN_FrontEndFloatModifier(Gameplay=0, Front End=0.5)
      B=&Lerp_Float(A=0, B=0.5, Alpha=$System.Timeline001.Time)
      Alpha=$System.Timeline.Time
     )
    )
    ScaleA=true
    ScaleRGB=true
   )
  ]
  rendererClasses [&NiagaraMeshRendererProperties]
  renderers [{rendererClass=&NiagaraMeshRendererProperties, rendererIndex=0}]
 }
}
summary {
 systemName=NS_Glider_AuraCop_Trail
 emitters [
  {emitterName=Local_Ribbon, rendererClasses [&NiagaraRibbonRendererProperties]}
  {emitterName=WindLines, rendererClasses [&NiagaraMeshRendererProperties]}
  {
   emitterName=NE_HeadWindCone
   rendererClasses [&NiagaraMeshRendererProperties, &NiagaraSpriteRendererProperties]
  }
  {emitterName=NE_SmokeTrail, rendererClasses [&NiagaraSpriteRendererProperties]}
  {emitterName=NE_SpiralTrails001, rendererClasses [&NiagaraMeshRendererProperties]}
 ]
 userVariables [
  User.bIsFrontEnd {type=NiagaraBool}
  User.bIsFrontEndPreview {type=NiagaraBool}
  User.bIsFullyDeployed {type=NiagaraBool}
  User.ForwardDot {type=NiagaraFloat, defaultValue=1, description=""}
  User.RightDot {type=NiagaraFloat, defaultValue=0, description=""}
  User.VariantIndex {type=NiagaraFloat, defaultValue=0, description=""}
 ]
}
