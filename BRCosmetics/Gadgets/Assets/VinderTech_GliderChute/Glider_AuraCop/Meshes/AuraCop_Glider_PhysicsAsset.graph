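# Physics asset for the AuraCop glider mesh, a cosmetic VinderTech glider chute.
# A single capsule body on the Frame bone, offset from the bone origin and rolled about 90 degrees onto its side, provides the whole collision shape.
# It has no constraints, so there is no jointed or multi-body ragdoll setup.
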PhysicsAsset /BRCosmetics/Gadgets/Assets/VinderTech_GliderChute/Glider_AuraCop/Meshes/AuraCop_Glider_PhysicsAsset
bodies {Frame [{shapeType=Capsule, length=94.097, radius=84.2942, center=<0.0029 -37.4223 -68.6065>, rotation=<rot 0 -179.9882 -89.7153>}]}
constraints []
